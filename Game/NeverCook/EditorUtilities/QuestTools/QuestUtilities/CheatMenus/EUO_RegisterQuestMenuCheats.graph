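# Editor utility that registers the quest cheat menu entries in the Unreal editor when run.
# Run waits for the Asset Registry, then recursively scans the QuestCheatMenuEntries folder for editor utility blueprints.
# It loads each ToolMenuEntryScript subclass, constructs it, sets its owner name and registers it as a menu entry.
# A timer-based registry retry event exists, but Run bypasses it in favor of the blocking wait.
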EditorUtilityBlueprint /Game/NeverCook/EditorUtilities/QuestTools/QuestUtilities/CheatMenus/EUO_RegisterQuestMenuCheats
graphs [EventGraph {nodes [
   n1 = CheckAssetRegistryAgain() {is=title K2Node_CustomEvent}
   > n2 = Branch(Condition=IsLoadingAssets!(self:Asset Registry Interface=GetAssetRegistry()) -> (ReturnValue:bool))
   > PrintString(InString=before timer)
   > Set Timer by Event(Delegate=@n1, Time=5, bLooping=false)
   > PrintString(InString=after timer)
   PrintString(execute=[@n9.then, @n2.else], InString=Registering Quest Cheat Menu Items)
   > WaitForCompletion(self:Asset Registry Interface=GetAssetRegistry())
   > n3 = GetAssetsByPath!(
    execute
    self:Asset Registry Interface=GetAssetRegistry()
    PackagePath:Name=/Game/Blueprints/Blutilities/ForDesign/Quests/CheatMenus/QuestCheatMenuEntries
    bRecursive:bool=true
    bIncludeOnlyOnDiskAssets:bool=false
   ) -> (then, OutAssetData:Array of Asset Data Structures, ReturnValue:bool)
   > n4 = For Each Loop(Array:Array of Asset Data Structures=@n3) -> (Array Element:Asset Data Structure)
   > Branch(Condition=ClassIsChildOf(
     TestClass=GetClass!(InAssetData:"Asset Data Structure (by ref)"=@n4) -> (ReturnValue:Object Class Reference)
     ParentClass=/Script/Blutility.EditorUtilityBlueprint
    ))
   > n5 = LoadBlueprintClass!(execute, AssetPath:String=GetFullName!(InAssetData:"Asset Data Structure (by ref)"=@n4) -> (ReturnValue:String)) -> (then, ReturnValue:Object Class Reference)
   > Branch(Condition=ClassIsChildOf(TestClass=@n5, ParentClass=/Script/ToolMenus.ToolMenuEntryScript))
   > n6 = Construct Object from Class(Class=@n5, self=Self-Reference()) -> (ReturnValue:Object Reference)
   > n7 = Cast To ToolMenuEntryScript!(execute, Object=@n6) -> (then, CastFailed, AsTool Menu Entry Script:Tool Menu Entry Script Object Reference) {is=title K2Node_DynamicCast}
   > n8 = Set members in Tool Menu Entry Script Data!(
    execute:Exec
    StructRef:"Tool Menu Entry Script Data Structure (by ref)"=Get Data!(self:Tool Menu Entry Script Object Reference=@n7) -> (Data:Tool Menu Entry Script Data Structure)
    OwnerName:Name=Get OwnerName() -> (OwnerName:Name=None)
   ) -> (then:Exec, StructOut:"Tool Menu Entry Script Data Structure (by ref)") {is=title K2Node_SetFieldsInStruct}
   > Set Data!(execute, Data:Tool Menu Entry Script Data Structure=@n8, self:Tool Menu Entry Script Object Reference=@n7) -> (then, Output_Get:Tool Menu Entry Script Data Structure)
   > RegisterMenuEntry(self:Tool Menu Entry Script Object Reference=@n7)
   > PrintString(InString=Append(A="Found and registered: ", B=GetDisplayName(Object=@n7)))
   n9 = Event Run()
  ]}]
variables [
 MenuEntriesToRegister {type="array<EditorUtilityToolMenuEntry>", default [null]}
 OwnerName {type=string, default=QuestEditorsBp}
]
